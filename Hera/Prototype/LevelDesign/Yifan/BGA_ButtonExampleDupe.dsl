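; Example interactable button Blueprint for Fortnite that can be pressed once, with custom prompt text and a debug print on interact.
; Interacting clears the can-be-pressed flag, then splits by authority: the server sets ButtonIsPressed, while clients play audio and a Niagara burst at the arrow.
; The OnRep for ButtonIsPressed toggles the point light's visibility.
; Server death calls an event whose remote branch spawns explosion effects and sound and hides the mesh.
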(event Interaction|EventBlueprintOnInteract
  (params (InteractingPawn :type Fort%20Pawn%20Object%20Reference) (InteractionBeingAttempted :type EInteractionBeingAttempted%20Enum :default "FirstInteraction"))
  ; Button Interaction Logic
  (Variables|Default|SetButtonCanBePressed false) ; set the button interaction to false, basically a once press button
  (CallFunction|ButtonPressedEvent)
  (Development|PrintString "Hello" true true "(R=0.000000,G=0.660000,B=1.000000,A=1.000000)" 5.0))

(event Custom|ButtonPressedEvent
  ; Button Interaction Event - What Happens?
  (Networking|SwitchHasAuthority
    (:Authority
      (Variables|Default|SetButtonIsPressed true) ; this changes the lights visibility on the clients
    )
    (:Remote
      (Audio|Components|Audio|Play (Variables|Default|GetButtonPressedAudio) 0.0) ; play the audio component  button press sound
      (Niagara|SpawnSystematLocation "/Game/Athena/Artemis/Environments/BuildingActors/Daisy/FX/NS_Daisy_Mushroom_Bounce.NS_Daisy_Mushroom_Bounce" (Transformation|GetWorldLocation (Variables|Default|GetArrow))) ; Spawn a new niagara particle effect at this location
    )) ; authority = server, Remote = Client
)

(event EventOnDeathServer
  (params (Damage :type Float%20%28single-precision%29) (DamageTags :type Gameplay%20Tag%20Container%20Structure%20%28by%20ref%29) (Momentum :type Vector :default "0, 0, 0") (HitInfo :type Hit%20Result%20Structure%20%28by%20ref%29) (InstigatedBy :type Controller%20Object%20Reference) (DamageCauser :type Actor%20Object%20Reference) (EffectContext :type Gameplay%20Effect%20Context%20Handle%20Structure))
  ; Button Destroyed Logic
  (CallFunction|PlayDeathFXAndSounds))

(event Custom|PlayDeathFXAndSounds
  ; Button Destroyed Logic
  (Networking|SwitchHasAuthority
    (:Remote
      (Niagara|SpawnSystematLocation "/Game/Athena/Items/Consumables/C4/NS_C4_Explosion_Athena.NS_C4_Explosion_Athena" (Transformation|GetActorLocation self)) ; Spawn a new niagara particle effect at this location
      (Audio|SpawnSoundatLocation "/Game/Athena/Sounds/Items/C4/C4_Explosion_Cue.C4_Explosion_Cue" (Transformation|GetActorLocation self)) ; a death sound
      (Rendering|SetVisibility (Variables|Default|GetStaticMesh) false false) ; this hides the mesh on the clients as its destroyed
    )))

(fn BlueprintCanInteract
  (params (InteractingPawn :type Fort%20Pawn%20Object%20Reference) (InteractionBeingAttempted :type EInteractionBeingAttempted%20Enum :default "FirstInteraction") (InteractionType :type TInteractionType%20Enum :default "IT_NoInteraction"))
  (return (Variables|Default|GetButtonCanBePressed)))

(fn BlueprintGetInteractionString
  (params (InteractingPawn :type Fort%20Pawn%20Object%20Reference) (InteractionBeingAttempted :type EInteractionBeingAttempted%20Enum :default "FirstInteraction"))
  (return (Variables|Default|GetButtonInteractionText)))

(fn OnRep_ButtonIsPressed
  (if (Variables|Default|GetButtonIsPressed)
    (Rendering|SetVisibility (Variables|Default|GetPointLight) true false)
    (else
      (Rendering|SetVisibility (Variables|Default|GetPointLight) false false))))

(fn ConstructionScript)
