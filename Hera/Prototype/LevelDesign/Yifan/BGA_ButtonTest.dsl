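; Prototype BGA interactable button, apparently a test, that can be pressed once and then plays a sound and effect.
; Interaction availability and prompt text come from the ButtonCanBePressed and ButtonText variables.
; On interact it clears ButtonCanBePressed, prints a debug message and calls ButtonPressEvent.
; That event plays ButtonSound and spawns a mushroom-bounce Niagara system at SpawnHere, only on the Remote (non-authority) branch.
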(event Interaction|EventBlueprintOnInteract
  (params (InteractingPawn :type Fort%20Pawn%20Object%20Reference) (InteractionBeingAttempted :type EInteractionBeingAttempted%20Enum :default "FirstInteraction"))
  ;[0.1324 0.05655 0.15 0.5] Add interaction button
  (Variables|Default|SetButtonCanBePressed false)
  (Development|PrintString "You pressed a button" true true "(R=0.000000,G=0.660000,B=1.000000,A=1.000000)" 3e+01)
  (CallFunction|ButtonPressEvent))

(event Custom|ButtonPressEvent
  ;[0.03371 0.1371 0.15 0.5] Spawn special effects and play music
  (Networking|SwitchHasAuthority
    (:Remote
      (Audio|Components|Audio|Play (Variables|Default|GetButtonSound) 0.0)
      (Niagara|SpawnSystematLocation "/Game/Athena/Artemis/Environments/BuildingActors/Daisy/FX/NS_Daisy_Mushroom_Bounce.NS_Daisy_Mushroom_Bounce" (Transformation|GetWorldLocation (Variables|Default|GetSpawnHere))))))

(fn BlueprintCanInteract
  (params (InteractingPawn :type Fort%20Pawn%20Object%20Reference) (InteractionBeingAttempted :type EInteractionBeingAttempted%20Enum :default "FirstInteraction") (InteractionType :type TInteractionType%20Enum :default "IT_NoInteraction"))
  (return (Variables|Default|GetButtonCanBePressed)))

(fn BlueprintGetInteractionString
  (params (InteractingPawn :type Fort%20Pawn%20Object%20Reference) (InteractionBeingAttempted :type EInteractionBeingAttempted%20Enum :default "FirstInteraction"))
  (return (Variables|Default|GetButtonText)))

(fn ConstructionScript)
