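; Prototype test actor whose trigger box damages player pawns that enter it.
; On BeginPlay it marks itself invulnerable and sets its can-be-damaged flag so the volume cannot be destroyed.
; On overlap, the authority casts the actor to FortPlayerPawnAthena, stores it, and applies fixed global environment damage unless the pawn is DBNO.
; Damage fires once per overlap begin, with no repeating tick.
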(event EventBeginPlay
  ; setup - trigger volume cannot be destroyed
  (Variables|BuildingActor|SetbIsInvulnerable true)
  (Fortnite|SetCanbeDamaged self))

(event OnComponentBeginOverlap(TriggerBox)
  (params (OverlappedComponent :type Primitive%20Component%20Object%20Reference) (OtherActor :type Actor%20Object%20Reference) (OtherComp :type Primitive%20Component%20Object%20Reference) (OtherBodyIndex :type Integer) (bFromSweep :type Boolean) (SweepResult :type Hit%20Result%20Structure%20%28by%20ref%29))
  ;[0.6979 0.3853 0.6705 0.5] Apply environmental damage to this player inside trigger
  (Networking|SwitchHasAuthority
    (:Authority
      (bind _asfort_player_pawn_athena (Utilities|Casting|CastToFortPlayerPawnAthena OtherActor))
      (bind _output_get (Variables|Default|SetThisPlayerPawn _asfort_player_pawn_athena))
      (if (not (Variables|/Script/FortniteGame.FortPawn|GetbIsDBNO _output_get))
        (Environment|ApplyGlobalEnvironmentDamagetoActor _output_get 5e+01)))))

(fn ConstructionScript)
