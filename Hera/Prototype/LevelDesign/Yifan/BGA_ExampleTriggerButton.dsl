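; Example trigger button, an interactable Blueprint that accepts a press until its can-be-pressed flag is cleared.
; On interact it clears that flag, calls ButtonPressEvent and prints a debug message.
; ButtonPressEvent sets ButtonIsPressed only with authority, and its OnRep shows the point light when true or the hologram otherwise.
; Interaction overrides return the can-be-pressed flag and a prompt string variable.
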(event EventBeginPlay)

(event Interaction|EventBlueprintOnInteract
  (params (InteractingPawn :type Fort%20Pawn%20Object%20Reference) (InteractionBeingAttempted :type EInteractionBeingAttempted%20Enum :default "FirstInteraction"))
  (Variables|Default|SetButtonCanbePressed false)
  (CallFunction|ButtonPressEvent)
  (Development|PrintString "Button is pressed" true true "(R=0.000000,G=0.660000,B=1.000000,A=1.000000)" 5.0))

(event Custom|ButtonPressEvent
  (Networking|SwitchHasAuthority
    (:Authority
      (Variables|Default|SetButtonIsPressed true))))

(fn BlueprintCanInteract
  (params (InteractingPawn :type Fort%20Pawn%20Object%20Reference) (InteractionBeingAttempted :type EInteractionBeingAttempted%20Enum :default "FirstInteraction") (InteractionType :type TInteractionType%20Enum :default "IT_NoInteraction"))
  (return (Variables|Default|GetButtonCanbePressed)))

(fn BlueprintGetInteractionString
  (params (InteractingPawn :type Fort%20Pawn%20Object%20Reference) (InteractionBeingAttempted :type EInteractionBeingAttempted%20Enum :default "FirstInteraction"))
  (return (Variables|Default|GetInteractionName)))

(fn OnRep_ButtonIsPressed
  (if (Variables|Default|GetButtonIsPressed)
    (Rendering|SetVisibility (Variables|Default|GetPointLight) true false)
    (Rendering|SetVisibility (Variables|Default|GetHologram) true false)))

(fn ConstructionScript)
