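; BGA_TriggerTest is an invulnerable prototype trigger actor that plays music while players are inside its TriggerBox.
; On authority, the first player pawn to enter sets SomeoneInsideTrigger and calls PlayMusic.
; When a player leaves and no player pawns still overlap, authority clears the flag and calls StopMusic.
; Both custom events act on the Music audio component only on the remote branch, and OnRep_SomeoneInsideTrigger is empty.
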(event EventBeginPlay
  ; Cannot destroy the trigger
  (Variables|BuildingActor|SetbIsInvulnerable true)
  (Fortnite|SetCanbeDamaged self))

(event OnComponentBeginOverlap(TriggerBox)
  (params (OverlappedComponent :type Primitive%20Component%20Object%20Reference) (OtherActor :type Actor%20Object%20Reference) (OtherComp :type Primitive%20Component%20Object%20Reference) (OtherBodyIndex :type Integer) (bFromSweep :type Boolean) (SweepResult :type Hit%20Result%20Structure%20%28by%20ref%29))
  ;[0.3802 0.342 0.1316 0.5] Start palying music when the player enters the trigger
  (Networking|SwitchHasAuthority
    (:Authority
      (Utilities|Casting|CastToPlayerPawn_Athena OtherActor)
      (if (not (Variables|Default|GetSomeoneInsideTrigger))
        (Variables|Default|SetSomeoneInsideTrigger true)
        (CallFunction|PlayMusic)))))

(event Custom|PlayMusic
  ;[0.799 0.4195 0.8333 0.5] Play music/stop music custom event for clients
  (Networking|SwitchHasAuthority
    (:Remote
      (Audio|Components|Audio|Play (Variables|Default|GetMusic) 0.0))))

(event Custom|StopMusic
  ;[0.799 0.4195 0.8333 0.5] Play music/stop music custom event for clients
  (Networking|SwitchHasAuthority
    (:Remote
      (Audio|Components|Audio|Stop (Variables|Default|GetMusic)))))

(event OnComponentEndOverlap(TriggerBox)
  (params (OverlappedComponent :type Primitive%20Component%20Object%20Reference) (OtherActor :type Actor%20Object%20Reference) (OtherComp :type Primitive%20Component%20Object%20Reference) (OtherBodyIndex :type Integer))
  ;[0.15 0.1407 0.04346 0.5] When player exit the trigger - check how many people are left inside, if = 0, stop playing music
  (Networking|SwitchHasAuthority
    (:Authority
      (Utilities|Casting|CastToPlayerPawn_Athena OtherActor)
      (if (== (Utilities|Array|Length (Collision|GetOverlappingActors (Variables|Default|GetTriggerBox) "/Script/FortniteGame.FortPlayerPawnAthena" :!class /Script/Engine.PrimitiveComponent)) 0)
        (Variables|Default|SetSomeoneInsideTrigger false)
        (CallFunction|StopMusic)))))

(fn OnRep_SomeoneInsideTrigger)

(fn ConstructionScript)
