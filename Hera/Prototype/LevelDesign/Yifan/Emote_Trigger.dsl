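; Invulnerable emote trigger volume that grants Slurpshroom shield effects to every Fortnite player inside while at least one of them is emoting.
; Once PartyHasStarted fires, a player entering the box makes the server check immediately and start a looping two-second timer that rechecks.
; The timer pauses when the box empties.
; IsAnyoneEmoting queries each overlapping pawn's controller, and the result feeds OnePlayerIsEmoting, whose OnRep only prints debug text.
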(event EventBeginPlay
  ; Invulnerable
  (Variables|BuildingActor|SetbIsInvulnerable true)
  (Variables|Actor|SetbCanBeDamaged false))

(event Custom|PartyHasStarted
  ;[0.1393 0.05578 0.15 0.5] Custom event to set party started boolean value to true
  (Variables|Default|SetPartyHasStrated true)
  (Development|PrintText "NSLOCTEXT(\"[FA9BA5984F8C4C6E039ED1E4DFC06A7C]\", \"642A4E064782C7355A18A79A4692B553\", \"PartyhasStarted set to true\")" true true "(R=0.000000,G=0.660000,B=1.000000,A=1.000000)" 1e+01))

(event OnComponentBeginOverlap(Box)
  (params (OverlappedComponent :type Primitive%20Component%20Object%20Reference) (OtherActor :type Actor%20Object%20Reference) (OtherComp :type Primitive%20Component%20Object%20Reference) (OtherBodyIndex :type Integer) (bFromSweep :type Boolean) (SweepResult :type Hit%20Result%20Structure%20%28by%20ref%29))
  (bind _timerhandle (Variables|Default|GetTimerhandle))
  ; Check if anyone is emoting inside the trigger
  (Networking|SwitchHasAuthority
    (:Authority
      (if (Variables|Default|GetPartyHasStrated)
        (Utilities|Casting|CastToFortPlayerPawnAthena OtherActor)
        (bind _returnvalue (Utilities|Time|SetTimerbyEvent (delegate Custom|CheckPlayerEmote) 2.0 true))
        (Variables|Default|SetTimerhandle _returnvalue)
        (bind _playerisemoting (CallFunction|IsAnyoneEmoting (Collision|GetOverlappingActors (Variables|Default|GetBox) "/Script/FortniteGame.FortPlayerPawnAthena" :!class /Script/Engine.PrimitiveComponent)))
        ; If no one is emoting, unpause the timer
        (bind _output_get (Variables|Default|SetOnePlayerIsEmoting _playerisemoting))
        (if _output_get
          ;[0.08313 0.8133 0.849 0.5] Apply Shield Regen as Gameplay effect to all players inside the trigger
          (for _array_element (Collision|GetOverlappingActors (Variables|Default|GetBox) "/Script/FortniteGame.FortPlayerPawnAthena" :!class /Script/Engine.PrimitiveComponent)
            (Environment|ApplyGlobalEnvironmentGameplayEffecttoActor _array_element "/Game/Athena/Items/EnvironmentalItems/EnvSlurp/GE_Athena_Slurpshroom_Shields.GE_Athena_Slurpshroom_Shields_C"))
          (elif (Utilities|Time|IsTimerPausedbyHandle _timerhandle)
            (Utilities|Time|UnpauseTimerbyHandle _timerhandle)))
        (else
          (Development|PrintText "NSLOCTEXT(\"[FA9BA5984F8C4C6E039ED1E4DFC06A7C]\", \"7C18D2294A2AAF976298718730FFCAA7\", \"Timer failed to start! \")" true true "(R=1.000000,G=0.866733,B=0.000000,A=1.000000)" 8.0))))))

(event Custom|CheckPlayerEmote
  (bind _timerhandle (Variables|Default|GetTimerhandle))
  ; Check if anyone is emoting inside the trigger
  (bind _playerisemoting (CallFunction|IsAnyoneEmoting (Collision|GetOverlappingActors (Variables|Default|GetBox) "/Script/FortniteGame.FortPlayerPawnAthena" :!class /Script/Engine.PrimitiveComponent)))
  ; If no one is emoting, unpause the timer
  (bind _output_get (Variables|Default|SetOnePlayerIsEmoting _playerisemoting))
  (if _output_get
    ;[0.08313 0.8133 0.849 0.5] Apply Shield Regen as Gameplay effect to all players inside the trigger
    (for _array_element (Collision|GetOverlappingActors (Variables|Default|GetBox) "/Script/FortniteGame.FortPlayerPawnAthena" :!class /Script/Engine.PrimitiveComponent)
      (Environment|ApplyGlobalEnvironmentGameplayEffecttoActor _array_element "/Game/Athena/Items/EnvironmentalItems/EnvSlurp/GE_Athena_Slurpshroom_Shields.GE_Athena_Slurpshroom_Shields_C"))
    (elif (Utilities|Time|IsTimerPausedbyHandle _timerhandle)
      (Utilities|Time|UnpauseTimerbyHandle _timerhandle))))

(event OnComponentEndOverlap(Box)
  (params (OverlappedComponent :type Primitive%20Component%20Object%20Reference) (OtherActor :type Actor%20Object%20Reference) (OtherComp :type Primitive%20Component%20Object%20Reference) (OtherBodyIndex :type Integer))
  ; Pause the timer when no one is inside the trigger
  (Networking|SwitchHasAuthority
    (:Authority
      (Utilities|Casting|CastToFortPlayerPawnAthena OtherActor)
      (if (== (Utilities|Array|Length (Collision|GetOverlappingActors (Variables|Default|GetBox) "/Script/FortniteGame.FortPlayerPawnAthena" :!class /Script/Engine.PrimitiveComponent)) 0)
        (Variables|Default|SetOnePlayerIsEmoting false)
        (Utilities|Time|PauseTimerbyHandle (Variables|Default|GetTimerhandle))))))

(fn isAnyoneEmoting
  (params (PlayerActors :type Array%20of%20Actor%20Object%20References))
  (bind _array_element (Utilities|Array|ForEachLoopwithBreak PlayerActors)
    (:LoopBody
      (bind _asfort_player_pawn_athena (Utilities|Casting|CastToFortPlayerPawnAthena _array_element))
      (bind _asfort_player_controller_athena (Utilities|Casting|CastToFortPlayerControllerAthena (Pawn|GetController _asfort_player_pawn_athena)))
      (if (Emote|IsPlayingEmote _asfort_player_controller_athena)
        (LocalVariables|SetIsEmotingLocal true)))
    (:Completed
      _))
  (return (LocalVariables|GetIsEmotingLocal)))

(fn OnRep_OnePlayerIsEmoting
  (if (Variables|Default|GetOnePlayerIsEmoting)
    (Development|PrintText "NSLOCTEXT(\"[FA9BA5984F8C4C6E039ED1E4DFC06A7C]\", \"6A1FFDE747211FA7859707B215BBFE51\", \"Player is emoting\")")))

(fn ConstructionScript)
